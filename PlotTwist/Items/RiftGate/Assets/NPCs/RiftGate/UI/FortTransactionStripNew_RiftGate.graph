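# Widget Blueprint for a Rift Gate NPC conversation transaction strip, derived from FortTransactionStrip, that shows a short uppercase cost description label.
# The inherited Text_CostDescription block sits right-aligned and vertically centered in a root Overlay.
# PreConstruct sets that label from the TransactionString variable, which defaults to "Cost", replacing the designer placeholder "Accept".
# The bFundingVersion and isReward flags are declared but unused in this graph.
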WidgetBlueprint /PlotTwist/Items/RiftGate/Assets/NPCs/RiftGate/UI/FortTransactionStripNew_RiftGate
graphs [EventGraph {nodes [
   Event PreConstruct()
   > "SetText (Text)"(
    self=Get Text_CostDescription() -> (Text_CostDescription:Common Text Object Reference)
    InText=Get TransactionString() -> (TransactionString:Text)
   )
  ]}]
variables [
 bFundingVersion {type=boolean}
 isReward {type=boolean}
 TransactionString {type=string, default=Cost}
]
widgets {
 info {inheritedWidgetCount=1, parentClass=/Script/FortniteConversationUI.FortTransactionStrip, rootWidgetClass=&Overlay, widgetCount=2}
 widgets [
  Overlay_57 {widgetClassPath=&Overlay, widget=Overlay_57}
  Text_CostDescription {
   widgetClassPath=&CommonTextBlock
   bInherited=true
   bIsVariable=true
   parent=Overlay_57
   widget=Text_CostDescription
   properties {
    text=Accept
    textTransformPolicy=ToUpper
    font {
     bForceMonospaced=false
     bMaterialIsStencil=false
     fontMaterial=null
     fontObject=&BurbankSmall
     letterSpacing=0
     monospacedWidth=1
     size=18
     skewAmount=0
     typefaceFontName=Black
     outlineSettings {
      bApplyOutlineToDropShadows=false
      bMiteredCorners=false
      bSeparateFillAlpha=false
      outlineMaterial=null
      outlineSize=0
      outlineColor=<rgba 0 0 0 1>
     }
    }
    margin=<ltrb 0 5 0 5>
    shadowOffset=<0 0>
    strikeBrush {
     bIsDynamicallyLoaded=false
     drawAs=Image
     imageType=NoImage
     mirroring=NoMirror
     resourceName=None
     resourceObject=null
     tiling=NoTile
     imageSize=<32 32>
     margin=<ltrb 0 0 0 0>
     outlineSettings {
      bUseBrushTransparency=false
      roundingType=HalfHeightRadius
      width=0
      color {colorUseRule=UseColor_Specified, specifiedColor=<rgba 0 0 0 0>}
      cornerRadii=<0 0 0 0>
     }
     tintColor {colorUseRule=UseColor_Specified, specifiedColor=<rgba 0 0 0 1>}
     uVRegion {bIsValid=false, max=<0 0>, min=<0 0>}
    }
   }
   slot {class=OverlaySlot, properties {horizontalAlignment=HAlign_Right, verticalAlignment=VAlign_Center}}
  }
 ]
}
