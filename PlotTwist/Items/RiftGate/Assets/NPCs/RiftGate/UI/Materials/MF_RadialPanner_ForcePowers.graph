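# Material function for Rift Gate UI materials that converts UVs to polar coordinates and pans them over time.
# Arctangent2 of the centred UV input gives a normalised angle scaled by Y Scale; VectorToRadialValue gives linear distance scaled by X Scale.
# The distance branch reads the default texture coordinates, ignoring the UV input.
# A static Flip UVs switch orders the two axes before a Panner driven by Time, X Speed and Y Speed inputs.
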MaterialFunction /PlotTwist/Items/RiftGate/Assets/NPCs/RiftGate/UI/Materials/MF_RadialPanner_ForcePowers
graph {
 properties {bExposeToLibrary=true}
 nodes [
  n1 = TextureCoordinate()
  n2 = ConstantBiasScale(None=FunctionInput(Preview=@n1) {bUsePreviewValueAsDefault=true, inputName=UV, inputType=FunctionInput_Vector2}) {bias=-0.5, scale=2}
  n3 = Arctangent2!(Y=ComponentMask(None=@n2) {r=true}, X=ComponentMask(None=@n2) {g=true})
  n4 = MaterialFunctionCall(Multiplier=Constant() {r=2}) {materialFunction=&Pi}
  n5 = Multiply(
   A=Add(A=Divide(A=@n3, B=@n4.Result)) {constB=0.5}
   B=FunctionInput(Preview=Constant() {r=1}) {bUsePreviewValueAsDefault=true, inputName=Y Scale, inputType=FunctionInput_Scalar}
  )
  n6 = Subtract(A=@n1) {constB=0.5}
  n7 = Constant() {r=1}
  n8 = ComponentMask(None=Multiply(
    A=MaterialFunctionCall(Vector or UVs=@n6) {materialFunction=&VectorToRadialValue}.Linear Distance
    B=FunctionInput(Preview=@n7) {bUsePreviewValueAsDefault=true, inputName=X Scale, inputType=FunctionInput_Scalar}
   )) {r=true}
  n9 = AppendVector(A=@n5, B=@n8)
  n10 = AppendVector(A=@n8, B=@n5)
  n11 = FunctionInput(Preview=StaticBool()) {bUsePreviewValueAsDefault=true, inputName=Flip UVs, inputType=FunctionInput_StaticBool}
  n12 = Time()
  n13 = FunctionInput() {bUsePreviewValueAsDefault=true, inputName=X Speed, inputType=FunctionInput_Scalar}
  n14 = FunctionInput() {bUsePreviewValueAsDefault=true, inputName=Y Speed, inputType=FunctionInput_Scalar}
  FunctionOutput(None=Panner(
    Coordinate=StaticSwitch(True=@n9, False=@n10, Value=@n11)
    Time=FunctionInput(Preview=@n12) {bUsePreviewValueAsDefault=true, inputName=Time, inputType=FunctionInput_Scalar}
    Speed=AppendVector(A=@n13, B=@n14)
   ) {speedX=1})
 ]
}
