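# Additive UI material for a Force Power icon background in RiftGate, drawing animated, tinted star patterns.
# Radially panned star textures combine with a panning noise mask, and the LightRays static switch can add a product of two counter-scrolling samples.
# StarColor tints the result, which a named-reroute mask multiplies.
# Exponential radial gradients form a ring outline mask, and a separate rotating radial-progress mask is also defined.
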Material /PlotTwist/Items/RiftGate/Assets/NPCs/RiftGate/UI/Materials/M_UI_ForcePower_Icon_Bg
graph {
 properties {blendMode=BLEND_Additive, materialDomain=MD_UI, pixelDepthOffsetMode=PDOM_Legacy, usedShadingModels=Default Lit}
 nodes [
  n1 = TextureSampleParameter2D(UVs=MaterialFunctionCall!(X Scale, Time, Y Scale, Flip UVs, X Speed, Y Speed=ScalarParameter() {defaultValue=-0.05, parameterName=6}, UV) {materialFunction=&/MF_RadialPanner_ForcePowers}.Result) {
   parameterName=02
   samplerType=SAMPLERTYPE_LinearColor
   texture=/PlotTwist/Items/RiftGate/Assets/NPCs/RiftGate/UI/Textures/T_UI_ForcePowers_Stars
  }
  n2 = ScalarParameter() {defaultValue=0.05, parameterName=3}
  n3 = TextureSampleParameter2D(UVs=MaterialFunctionCall(X Scale=ScalarParameter() {defaultValue=1, parameterName=1}, Y Scale=ScalarParameter() {defaultValue=4, parameterName="2)"}, Y Speed=@n2) {materialFunction=&MF_RadialPanner_ForcePowers}.Result) {
   parameterName=02
   samplerType=SAMPLERTYPE_LinearColor
   texture=&T_UI_ForcePowers_Stars
  }
  Add(A=@n1.B, B=@n3.B)
  Multiply(A=@n2, B=Time())
  n4 = ScalarParameter() {defaultValue=0.3, parameterName=StarMask_XSpeed}
  n5 = ScalarParameter() {defaultValue=0.1, parameterName=StarMask_YSpeed}
  n6 = TextureSampleParameter2D(UVs=Panner(Coordinate=TextureCoordinate() {uTiling=0.7, vTiling=0.2}, Speed=AppendVector(A=@n4, B=@n5)) {speedY=0.5}) {
   parameterName=01
   samplerType=SAMPLERTYPE_LinearGrayscale
   texture=/Game/Characters/Player/Female/Medium/Bodies/F_MED_Cube_Queen/Textures/TechArt/T_CubeQueenNoiseA
  }
  n7 = ScalarParameter() {defaultValue=0.1, parameterName="Star Tiling (Vertical Scale)"}
  n8 = ScalarParameter() {defaultValue=2, parameterName=Star Speed}
  n9 = TextureSampleParameter2D(UVs=MaterialFunctionCall(X Scale=@n7, Y Scale=ScalarParameter() {defaultValue=1, parameterName="Star Tiling (Horizontal)"}, X Speed=Multiply(A=@n7, B=@n8)) {materialFunction=&MF_RadialPanner_ForcePowers}.Result) {
   parameterName=02
   samplerType=SAMPLERTYPE_LinearColor
   texture=&T_UI_ForcePowers_Stars
  }
  n10 = Add(A=Multiply(A=@n6.G, B=@n9.G), B=Add(A=@n9.R, B=@n9.R))
  n11 = StaticSwitchParameter(True=Multiply(A=@n1.B, B=@n3.B), False=@n10) {parameterName=LightRays}
  Multiply(A=Multiply(A=VectorParameter() {parameterName=StarColor, defaultValue=<rgba 1 1 1 1>}.RGB, B=Add(A=@n10, B=@n11)), B=NamedRerouteUsage())
  n12 = ScalarParameter() {defaultValue=0.5, group=MaskReveal, parameterName=OutlineThicknessOut}
  n13 = ScalarParameter() {defaultValue=3, group=MaskReveal, parameterName=OutlineThicknessIn}
  n14 = MaterialFunctionCall(
   UVs=TextureCoordinate()
   Radius=ScalarParameter() {defaultValue=0.35, group=MaskReveal, parameterName=InnerMaskRevealRadius}
   Density=ScalarParameter() {defaultValue=3, group=MaskReveal, parameterName=InnerMaskRevealDensity}
  ) {materialFunction=&RadialGradientExponential}
  NamedRerouteDeclaration(None=Multiply(
    A=MaterialFunctionCall(Radius=@n12, Density=@n13) {materialFunction=&RadialGradientExponential}.RadialGradientExponential
    B=OneMinus(None=@n14.RadialGradientExponential)
   )) {name=CircleOutline, nodeColor=<rgba 1 0 0.2109 1>}
  n15 = Rotator(Time=Multiply(A=Time(), B=ScalarParameter() {parameterName=Rotation Speed, sortPriority=5}))
  n16 = ScalarParameter() {parameterName=EdgeSofteness}
  n17 = ScalarParameter() {defaultValue=1, parameterName=FlameRadialMask, sortPriority=4}
  NamedRerouteDeclaration(None=Multiply(
    A=NamedRerouteUsage()
    B=MaterialFunctionCall!(UVs=@n15, EdgeSoftness=@n16, Progress=@n17) {materialFunction=/Game/UI/Foundation/Toasts/Materials/MF_RadialMask}.Mask
   )) {name=RadialMask, nodeColor=<rgba 0 0.8359 1 1>}
  Rotator()
 ]
}
