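# Rift Gate NPC conversation brief widget that presents a purchasable Force power service with icon, title, description and transaction strip.
# A visibility switcher swaps between that panel and an error message showing the reason purchase is unavailable.
# On configure, it IsA-matches the considered task node against a class-to-icon map, using the key index to read the values array.
# The first match sets and reveals the soft-material icon and tints IconFX's dynamic material.
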WidgetBlueprint /PlotTwist/Items/RiftGate/Assets/NPCs/RiftGate/UI/FortDataDrivenServiceBrief_RiftGate
graphs [EventGraph {nodes [
   Event Construct()
   > Parent: Construct()
   > SetVisibility(self=Get Image_Icon() -> (Image_Icon:Image Object Reference), InVisibility=Hidden)
   > FortShippingLog!(
    execute
    WorldContextObject:Object Reference=Self-Reference()
    InString:String="FortDataDrivenServiceBrief_RiftGate::OnConstruct() setting image to hidden."
    bLogAsWarning:bool=false
   ) -> (then)
   Event OnAbleToPurchase() {is=title K2Node_Event}
   > SetActiveWidget(
    self=Get VisibilitySwitcher() -> (VisibilitySwitcher:Common Visibility Switcher Object Reference)
    Widget=Get CanPurchase() -> (CanPurchase:Overlay Object Reference)
   )
   n1 = Event OnUnableToPurchase!() -> (OutputDelegate, then, Reason:"Text (by ref)") {is=title K2Node_Event}
   > SetActiveWidget(self=Get VisibilitySwitcher(), Widget=Get CannotPurchase() -> (CannotPurchase:Overlay Object Reference))
   > SetErrorMessage!(
    execute
    self:Brief Error Message New Object Reference=Get BriefErrorMessage() -> (BriefErrorMessage:Brief Error Message New Object Reference)
    ErrorText:Text=@n1.Reason
   ) -> (then)
   n2 = Event OnConfigure!() -> (
    OutputDelegate
    then
    ClientContext:"Conversation Context Structure (by ref)"
    OptionEntry:"Client Conversation Option Entry Structure (by ref)"
    SelectedIndex:int=0
   ) {is=title K2Node_Event}
   > n3 = GetTaskBeingConsidered!(execute, Context:"Conversation Context Structure (by ref)"=@n2.ClientContext) -> (then, ReturnValue:Conversation Task Node Object Reference)
   n4 = Get TaskNodeIconDataMap Const() -> (TaskNodeIconDataMap Const:Map of Fort Conversation Task Node Data Driven Service Class References to STRUCT Rift Gate Icon Data Structures)
   > n5 = Values!(execute, TargetMap:Map of Fort Conversation Task Node Data Driven Service Class References to STRUCT Rift Gate Icon Data Structures=@n4) -> (then, Values:Array of STRUCT Rift Gate Icon Data Structures)
   > n6 = Keys!(execute, TargetMap:Map of Fort Conversation Task Node Data Driven Service Class References to STRUCT Rift Gate Icon Data Structures=@n4) -> (then, Keys:Array of Fort Conversation Task Node Data Driven Service Class References)
   > n7 = For Each Loop with Break!(Exec:Exec, Array:Array of Fort Conversation Task Node Data Driven Service Class References=@n6, Break:Exec=[@n11.Is Not Valid, @n12.then]) -> (LoopBody:Exec, Array Element:Fort Conversation Task Node Data Driven Service Class Reference, Array Index:int, Completed:Exec)
   > Branch(Condition=IsA!(Object:Object Reference=@n3, ObjectClass:Object Class Reference=@n7) -> (ReturnValue:bool))
   > FortShippingLog(
    WorldContextObject=Self-Reference()
    InString=Append!(
     A="FortDataDrivenServiceBrief_RiftGate::EventOnConfigure() setting icon based on task node class: ["
     B=Get Class Display Name(Class=@n7)
     C:String="]."
    ) -> (ReturnValue)
    bLogAsWarning=false
   )
   n8 = Break STRUCT Rift Gate Icon Data!(STRUCT_RiftGateIconData:"STRUCT Rift Gate Icon Data Structure (by ref)"="Get (a ref)"(Array:Array of STRUCT Rift Gate Icon Data Structures=@n5, Dimension 1=@n7.Array Index) -> (Output:"STRUCT Rift Gate Icon Data Structure (by ref)")) -> (
    SoftIconMaterialInstance_5_#1:Material Instance Soft Object Reference
    MaterialColor_4_#2:Linear Color Structure
   ) {is=title K2Node_BreakStruct}
   n9 = Get Image_Icon()
   > SetBrushFromSoftMaterial!(execute, self:Image Object Reference=@n9, SoftMaterial:Material Interface Soft Object Reference=@n8) -> (then)
   > SetVisibility(self=@n9, InVisibility=Visible)
   > n10 = GetDynamicMaterial(self=Get Image_IconFX() -> (Image_IconFX:Image Object Reference))
   > n11 = Is Valid(InputObject=@n10)
   > n12 = SetVectorParameterValue(
    self=@n10
    ParameterName=Get BackgroundColorParam() -> (BackgroundColorParam:Name=None)
    Value=@n8.MaterialColor_4_#2
   )
  ]}]
variables [
 BackgroundColorParam {type=string, default=None}
 TaskNodeIconDataMap Const {default {
   /Script/Engine.BlueprintGeneratedClass'/PlotTwist/Items/RiftGate/Assets/NPCs/RiftGate/RiftGate_TaskNode_GainAmberVine.RiftGate_TaskNode_GainAmberVine_C' {
    soft Icon Material Instance=/PlotTwist/Items/RiftGate/Assets/NPCs/RiftGate/UI/Materials/MI_UI_Icon_FP_AmberVine
    material Color=<rgba 0 0 0 0>
   }
   &BlueprintGeneratedClass'/PlotTwist/Items/RiftGate/Assets/NPCs/RiftGate/RiftGate_TaskNode_GainForceProjectile.RiftGate_TaskNode_GainForceProjectile_C' {soft Icon Material Instance=/PlotTwist/Items/RiftGate/Assets/NPCs/RiftGate/UI/Materials/MI_UI_Icon_FP_Throw, material Color=<rgba 0 0 0 0>}
   &BlueprintGeneratedClass'/PlotTwist/Items/RiftGate/Assets/NPCs/RiftGate/RiftGate_TaskNode_GainForcePush.RiftGate_TaskNode_GainForcePush_C' {soft Icon Material Instance=/PlotTwist/Items/RiftGate/Assets/NPCs/RiftGate/UI/Materials/MI_UI_Icon_FP_Push, material Color=<rgba 0 0 0 0>}
   &BlueprintGeneratedClass'/PlotTwist/Items/RiftGate/Assets/NPCs/RiftGate/RiftGate_TaskNode_GainForceRip.RiftGate_TaskNode_GainForceRip_C' {soft Icon Material Instance=/PlotTwist/Items/RiftGate/Assets/NPCs/RiftGate/UI/Materials/MI_UI_Icon_FP_Pull, material Color=<rgba 0 0 0 0>}
   &BlueprintGeneratedClass'/PlotTwist/Items/RiftGate/Assets/NPCs/RiftGate/RiftGate_TaskNode_GainForceWht.RiftGate_TaskNode_GainForceWht_C' {soft Icon Material Instance=&MI_UI_Icon_FP_Push, material Color=<rgba 0 0 0 0>}
   &BlueprintGeneratedClass'/PlotTwist/Items/RiftGate/Assets/NPCs/RiftGate/RiftGate_TaskNode_GainHeavyLamp.RiftGate_TaskNode_GainHeavyLamp_C' {
    soft Icon Material Instance=/PlotTwist/Items/RiftGate/Assets/NPCs/RiftGate/UI/Materials/MI_UI_Icon_FP_HeavyLamp
    material Color=<rgba 0 0 0 0>
   }
   &BlueprintGeneratedClass'/PlotTwist/Items/RiftGate/Assets/NPCs/RiftGate/RiftGate_TaskNode_GainThunderBarnPull.RiftGate_TaskNode_GainThunderBarnPull_C' {soft Icon Material Instance=&MI_UI_Icon_FP_Pull, material Color=<rgba 0 0 0 0>}
   &BlueprintGeneratedClass'/PlotTwist/Items/RiftGate/Assets/NPCs/RiftGate/RiftGate_TaskNode_GainThunderBarnPush.RiftGate_TaskNode_GainThunderBarnPush_C' {soft Icon Material Instance=&MI_UI_Icon_FP_Push, material Color=<rgba 0 0 0 0>}
  }}
]
widgets {
 info {
  inheritedWidgetCount=4
  parentClass=/Script/FortniteConversationUI.FortDataDrivenServiceBrief
  rootWidgetClass=&SizeBox
  widgetCount=23
 }
 widgets [
  SizeBox_28 {
   widgetClassPath=&SizeBox
   widget=SizeBox_28
   properties {bOverride_WidthOverride=true, heightOverride=350, maxDesiredWidth=900, widthOverride=700}
  }
  VisibilitySwitcher {
   widgetClassPath=&CommonVisibilitySwitcher
   bIsVariable=true
   parent=SizeBox_28
   widget=VisibilitySwitcher
   slot {class=SizeBoxSlot}
  }
  CanPurchase {
   widgetClassPath=&Overlay
   bIsVariable=true
   parent=VisibilitySwitcher
   widget=CanPurchase
   slot {class=CommonVisibilitySwitcherSlot}
  }
  Image_ShadowBg {
   widgetClassPath=&Image
   bIsVariable=true
   parent=CanPurchase
   widget=Image_ShadowBg
   properties {
    brush {
     bIsDynamicallyLoaded=false
     drawAs=RoundedBox
     imageType=NoImage
     mirroring=NoMirror
     resourceName=None
     resourceObject=/PlotTwist/Items/RiftGate/Assets/NPCs/RiftGate/UI/Materials/MI_UI_ForcePower_ShadowBg
     tiling=NoTile
     imageSize=<32 32>
     margin=<ltrb 0.5 0.5 0.5 0.5>
     outlineSettings {
      bUseBrushTransparency=false
      roundingType=HalfHeightRadius
      width=0
      color {colorUseRule=UseColor_Specified, specifiedColor=<rgba 0 0 0 0>}
      cornerRadii=<0 0 0 0>
     }
     tintColor {colorUseRule=UseColor_Specified, specifiedColor=<rgba 0 0.0128 0.154 1>}
     uVRegion {bIsValid=false, max=<0 0>, min=<0 0>}
    }
    colorAndOpacity=<rgba 1 1 1 0>
    renderTransform {angle=0, scale=<1.05 0.55>, shear=<0 0>, translation=<0 0>}
   }
   slot {class=OverlaySlot, properties {horizontalAlignment=HAlign_Fill, verticalAlignment=VAlign_Fill}}
  }
  Image_LightBg {
   widgetClassPath=&Image
   bIsVariable=true
   parent=CanPurchase
   widget=Image_LightBg
   properties {
    brush {
     bIsDynamicallyLoaded=false
     drawAs=Box
     imageType=NoImage
     mirroring=NoMirror
     resourceName=None
     resourceObject=/PlotTwist/Items/RiftGate/Assets/NPCs/RiftGate/UI/Materials/MI_UI_ForcePower_LightBg
     tiling=NoTile
     imageSize=<32 32>
     margin=<ltrb 0.5 0.5 0.5 0.5>
     outlineSettings {
      bUseBrushTransparency=false
      roundingType=HalfHeightRadius
      width=0
      color {colorUseRule=UseColor_Specified, specifiedColor=<rgba 0 0 0 0>}
      cornerRadii=<0 0 0 0>
     }
     tintColor {colorUseRule=UseColor_Specified, specifiedColor=<rgba 1 1 1 1>}
     uVRegion {bIsValid=false, max=<0 0>, min=<0 0>}
    }
    colorAndOpacity=<rgba 1 1 1 0.08>
    renderTransform {angle=0, scale=<1 1>, shear=<0 0>, translation=<0 5>}
   }
   slot {class=OverlaySlot, properties {horizontalAlignment=HAlign_Fill, verticalAlignment=VAlign_Fill}}
  }
  Image_TextBox {
   widgetClassPath=&Image
   bIsVariable=true
   parent=CanPurchase
   widget=Image_TextBox
   properties {
    brush {
     bIsDynamicallyLoaded=false
     drawAs=Image
     imageType=NoImage
     mirroring=NoMirror
     resourceName=None
     resourceObject=/PlotTwist/Items/RiftGate/Assets/NPCs/RiftGate/UI/Materials/MI_UI_ForcePower_RoundedBox
     tiling=NoTile
     imageSize=<32 32>
     margin=<ltrb 0.5 0.5 0.5 0.5>
     outlineSettings {
      bUseBrushTransparency=false
      roundingType=HalfHeightRadius
      width=0
      color {colorUseRule=UseColor_Specified, specifiedColor=<rgba 0 0 0 0>}
      cornerRadii=<0 0 0 0>
     }
     tintColor {colorUseRule=UseColor_Specified, specifiedColor=<rgba 0 0.0139 0.0409 1>}
     uVRegion {bIsValid=false, max=<0 0>, min=<0 0>}
    }
    colorAndOpacity=<rgba 1 1 1 0.72>
    renderTransform {angle=0, scale=<1.05 1>, shear=<0 0>, translation=<0 0>}
   }
   slot {class=OverlaySlot, properties {horizontalAlignment=HAlign_Fill, verticalAlignment=VAlign_Fill, padding=<ltrb 0 32 0 24>}}
  }
  Image_Frame_Bottom {
   widgetClassPath=&Image
   bIsVariable=true
   parent=CanPurchase
   widget=Image_Frame_Bottom
   properties {
    brush {
     bIsDynamicallyLoaded=false
     drawAs=Image
     imageType=NoImage
     mirroring=NoMirror
     resourceName=None
     resourceObject=/PlotTwist/Items/RiftGate/Assets/NPCs/RiftGate/UI/Materials/MI_UI_Frame
     tiling=NoTile
     imageSize=<32 32>
     margin=<ltrb 0 0 0 0>
     outlineSettings {
      bUseBrushTransparency=false
      roundingType=HalfHeightRadius
      width=0
      color {colorUseRule=UseColor_Specified, specifiedColor=<rgba 0 0 0 0>}
      cornerRadii=<0 0 0 0>
     }
     tintColor {colorUseRule=UseColor_Specified, specifiedColor=<rgba 0.0212 0.0742 0.1248 1>}
     uVRegion {bIsValid=false, max=<0 0>, min=<0 0>}
    }
    colorAndOpacity=<rgba 1 1 1 0.4>
    renderTransform {angle=0, scale=<1.05 -1.5>, shear=<0 0>, translation=<0 -10>}
   }
   slot {class=OverlaySlot, properties {horizontalAlignment=HAlign_Fill, verticalAlignment=VAlign_Bottom}}
  }
  Image_Frame_Top {
   widgetClassPath=&Image
   bIsVariable=true
   parent=CanPurchase
   widget=Image_Frame_Top
   properties {
    brush {
     bIsDynamicallyLoaded=false
     drawAs=Image
     imageType=NoImage
     mirroring=NoMirror
     resourceName=None
     resourceObject=&MI_UI_Frame
     tiling=NoTile
     imageSize=<32 32>
     margin=<ltrb 0 0 0 0>
     outlineSettings {
      bUseBrushTransparency=false
      roundingType=HalfHeightRadius
      width=0
      color {colorUseRule=UseColor_Specified, specifiedColor=<rgba 0 0 0 0>}
      cornerRadii=<0 0 0 0>
     }
     tintColor {colorUseRule=UseColor_Specified, specifiedColor=<rgba 0.0212 0.0742 0.1248 1>}
     uVRegion {bIsValid=false, max=<0 0>, min=<0 0>}
    }
    colorAndOpacity=<rgba 1 1 1 0.4>
    renderTransform {angle=0, scale=<1.05 1.5>, shear=<0 0>, translation=<0 10>}
   }
   slot {class=OverlaySlot, properties {horizontalAlignment=HAlign_Fill, padding=<ltrb 0 10 0 0>}}
  }
  HorizontalBox_0 {
   widgetClassPath=&HorizontalBox
   parent=CanPurchase
   widget=HorizontalBox_0
   slot {class=OverlaySlot, properties {horizontalAlignment=HAlign_Fill, verticalAlignment=VAlign_Fill}}
  }
  Overlay_2 {
   widgetClassPath=&Overlay
   parent=HorizontalBox_0
   widget=Overlay_2
   properties {renderTransform {angle=0, scale=<1.2 1.2>, shear=<0 0>, translation=<0 0>}}
   slot {class=HorizontalBoxSlot, properties {horizontalAlignment=HAlign_Left, verticalAlignment=VAlign_Center, padding=<ltrb 25 0 25 0>}}
  }
  SizeBox {
   widgetClassPath=&SizeBox
   parent=Overlay_2
   widget=SizeBox
   properties {bOverride_HeightOverride=true, bOverride_WidthOverride=true, heightOverride=100, widthOverride=100}
   slot {class=OverlaySlot, properties {padding=<ltrb 0 10 0 0>}}
  }
  Image_Icon {
   widgetClassPath=&Image
   bIsVariable=true
   parent=SizeBox
   widget=Image_Icon
   properties {brush {
     bIsDynamicallyLoaded=false
     drawAs=Image
     imageType=NoImage
     mirroring=NoMirror
     resourceName=None
     resourceObject=null
     tiling=NoTile
     imageSize=<128 128>
     margin=<ltrb 0 0 0 0>
     outlineSettings {
      bUseBrushTransparency=false
      roundingType=HalfHeightRadius
      width=0
      color {colorUseRule=UseColor_Specified, specifiedColor=<rgba 0 0 0 0>}
      cornerRadii=<0 0 0 0>
     }
     tintColor {colorUseRule=UseColor_Specified, specifiedColor=<rgba 1 1 1 1>}
     uVRegion {bIsValid=false, max=<0 0>, min=<0 0>}
    }}
   slot {class=SizeBoxSlot}
  }
  Image_IconFX {
   widgetClassPath=&Image
   bIsVariable=true
   parent=Overlay_2
   widget=Image_IconFX
   properties {
    visibility=SelfHitTestInvisible
    brush {
     bIsDynamicallyLoaded=false
     drawAs=Image
     imageType=NoImage
     mirroring=NoMirror
     resourceName=None
     resourceObject=/PlotTwist/Items/RiftGate/Assets/NPCs/RiftGate/UI/Materials/M_UI_ForcePower_Icon_Bg
     tiling=NoTile
     imageSize=<32 32>
     margin=<ltrb 0 0 0 0>
     outlineSettings {
      bUseBrushTransparency=false
      roundingType=HalfHeightRadius
      width=0
      color {colorUseRule=UseColor_Specified, specifiedColor=<rgba 0 0 0 0>}
      cornerRadii=<0 0 0 0>
     }
     tintColor {colorUseRule=UseColor_Specified, specifiedColor=<rgba 0 0.2413 1 1>}
     uVRegion {bIsValid=false, max=<0 0>, min=<0 0>}
    }
    colorAndOpacity=<rgba 1 1 1 0.1>
    renderTransform {angle=0, scale=<2.5 2>, shear=<0 0>, translation=<0 0>}
   }
   slot {class=OverlaySlot, properties {horizontalAlignment=HAlign_Fill, verticalAlignment=VAlign_Fill}}
  }
  Overlay_Content {
   widgetClassPath=&Overlay
   parent=HorizontalBox_0
   widget=Overlay_Content
   slot {class=HorizontalBoxSlot, properties {size {sizeRule=Fill, value=1}}}
  }
  VerticalBox_Text {
   widgetClassPath=&VerticalBox
   parent=Overlay_Content
   widget=VerticalBox_Text
   slot {class=OverlaySlot, properties {horizontalAlignment=HAlign_Fill, verticalAlignment=VAlign_Fill, padding=<ltrb 10 0 0 0>}}
  }
  Text_TitleBar {
   widgetClassPath=&CommonTextBlock
   bInherited=true
   parent=VerticalBox_Text
   widget=Text_TitleBar
   properties {
    text=ServiceName
    textTransformPolicy=ToUpper
    font {
     bForceMonospaced=false
     bMaterialIsStencil=false
     fontMaterial=/PlotTwist/Items/RiftGate/Assets/NPCs/RiftGate/UI/Materials/M_Font_ForcePower_FontGradient
     fontObject=&Burbank
     letterSpacing=0
     monospacedWidth=1
     size=24
     skewAmount=0
     typefaceFontName=BurbankRegular-Black
     outlineSettings {
      bApplyOutlineToDropShadows=false
      bMiteredCorners=false
      bSeparateFillAlpha=false
      outlineMaterial=null
      outlineSize=5
      outlineColor=<rgba 0 0.0102 0.0409 1>
     }
    }
    shadowColorAndOpacity=<rgba 0 0 0 1>
    shadowOffset=<0 0>
    strikeBrush {
     bIsDynamicallyLoaded=false
     drawAs=Image
     imageType=NoImage
     mirroring=NoMirror
     resourceName=None
     resourceObject=null
     tiling=NoTile
     imageSize=<216.7805 105.8215>
     margin=<ltrb 0 0 0 0>
     outlineSettings {
      bUseBrushTransparency=false
      roundingType=HalfHeightRadius
      width=0
      color {colorUseRule=UseColor_Specified, specifiedColor=<rgba 0 0 0 0>}
      cornerRadii=<0 0 0 0>
     }
     tintColor {colorUseRule=UseColor_Specified, specifiedColor=<rgba 1 1 1 1>}
     uVRegion {bIsValid=false, max=<0 0>, min=<0 0>}
    }
   }
   slot {class=VerticalBoxSlot, properties {horizontalAlignment=HAlign_Left, verticalAlignment=VAlign_Top}}
  }
  RichText_Description {
   widgetClassPath=&CommonRichTextBlock
   bInherited=true
   bIsVariable=true
   parent=VerticalBox_Text
   widget=RichText_Description
   properties {
    autoWrapText=true
    bOverrideDefaultStyle=true
    text="Learn <Highlight>Force [POWER]</> and <Highlight>Force Jump</> while using a <Highlight>Lightsaber</> "
    textStyleSet=&/Conversation_RichTextStyleSet_RiftGate
    defaultTextStyleOverride {
     overflowPolicy=Clip
     transformPolicy=None
     colorAndOpacity {colorUseRule=UseColor_Specified, specifiedColor=<rgba 1 1 1 1>}
     font {
      bForceMonospaced=false
      bMaterialIsStencil=false
      fontMaterial=null
      fontObject=&BurbankSmall
      letterSpacing=0
      monospacedWidth=1
      size=20
      skewAmount=0
      typefaceFontName=Bold
      outlineSettings {
       bApplyOutlineToDropShadows=false
       bMiteredCorners=false
       bSeparateFillAlpha=false
       outlineMaterial=null
       outlineSize=0
       outlineColor=<rgba 0 0 0 1>
      }
     }
     highlightColor {colorUseRule=UseColor_Specified, specifiedColor=<rgba 0 0 0 1>}
     highlightShape {
      bIsDynamicallyLoaded=false
      drawAs=Image
      imageType=NoImage
      mirroring=NoMirror
      resourceName=None
      resourceObject=null
      tiling=NoTile
      imageSize=<32 32>
      margin=<ltrb 0 0 0 0>
      outlineSettings {
       bUseBrushTransparency=false
       roundingType=HalfHeightRadius
       width=0
       color {colorUseRule=UseColor_Specified, specifiedColor=<rgba 0 0 0 0>}
       cornerRadii=<0 0 0 0>
      }
      tintColor {colorUseRule=UseColor_Specified, specifiedColor=<rgba 1 1 1 1>}
      uVRegion {bIsValid=false, max=<0 0>, min=<0 0>}
     }
     selectedBackgroundColor {colorUseRule=UseColor_Foreground, specifiedColor=<rgba 1 0 1 1>}
     shadowColorAndOpacity=<rgba 0 0 0 0>
     shadowOffset=<0 0>
     strikeBrush {
      bIsDynamicallyLoaded=false
      drawAs=Image
      imageType=NoImage
      mirroring=NoMirror
      resourceName=None
      resourceObject=null
      tiling=NoTile
      imageSize=<32 32>
      margin=<ltrb 0 0 0 0>
      outlineSettings {
       bUseBrushTransparency=false
       roundingType=HalfHeightRadius
       width=0
       color {colorUseRule=UseColor_Specified, specifiedColor=<rgba 0 0 0 0>}
       cornerRadii=<0 0 0 0>
      }
      tintColor {colorUseRule=UseColor_Specified, specifiedColor=<rgba 1 1 1 1>}
      uVRegion {bIsValid=false, max=<0 0>, min=<0 0>}
     }
     underlineBrush {
      bIsDynamicallyLoaded=false
      drawAs=Image
      imageType=NoImage
      mirroring=NoMirror
      resourceName=None
      resourceObject=null
      tiling=NoTile
      imageSize=<32 32>
      margin=<ltrb 0 0 0 0>
      outlineSettings {
       bUseBrushTransparency=false
       roundingType=HalfHeightRadius
       width=0
       color {colorUseRule=UseColor_Specified, specifiedColor=<rgba 0 0 0 0>}
       cornerRadii=<0 0 0 0>
      }
      tintColor {colorUseRule=UseColor_Specified, specifiedColor=<rgba 1 1 1 1>}
      uVRegion {bIsValid=false, max=<0 0>, min=<0 0>}
     }
    }
    renderTransform {angle=0, scale=<1 1>, shear=<0 0>, translation=<0 -20>}
   }
   slot {class=VerticalBoxSlot, properties {padding=<ltrb 0 17 30 0>}}
  }
  TransactionStrip_Main {
   widgetClassPath=&/FortTransactionStripNew_RiftGate
   bInherited=true
   bIsVariable=true
   parent=VerticalBox_Text
   widget=TransactionStrip_Main
   properties {renderTransform {angle=0, scale=<1 1>, shear=<0 0>, translation=<0 -24>}}
   slot {class=VerticalBoxSlot, properties {horizontalAlignment=HAlign_Right, verticalAlignment=VAlign_Bottom, padding=<ltrb 0 0 35 5>}}
  }
  Text_StockRemaining {
   widgetClassPath=&CommonTextBlock
   bInherited=true
   bIsVariable=true
   parent=CanPurchase
   widget=Text_StockRemaining
   properties {
    justification=Center
    text=Stock: 1
    textTransformPolicy=ToUpper
    visibility=Collapsed
    font {
     bForceMonospaced=false
     bMaterialIsStencil=false
     fontMaterial=null
     fontObject=&Burbank
     letterSpacing=0
     monospacedWidth=1
     size=14
     skewAmount=0
     typefaceFontName=Burbank-PlayerPower
     outlineSettings {
      bApplyOutlineToDropShadows=false
      bMiteredCorners=false
      bSeparateFillAlpha=false
      outlineMaterial=&M-FontShadow
      outlineSize=2
      outlineColor=<rgba 0.0052 0.0052 0.0056 0.8>
     }
    }
    renderTransform {angle=0, scale=<1 1>, shear=<-8 0>, translation=<0 0>}
    shadowOffset=<0 0>
   }
   slot {class=OverlaySlot, properties {horizontalAlignment=HAlign_Right, verticalAlignment=VAlign_Bottom}}
  }
  Image_FrontGlow {
   widgetClassPath=&Image
   bIsVariable=true
   parent=CanPurchase
   widget=Image_FrontGlow
   properties {
    brush {
     bIsDynamicallyLoaded=false
     drawAs=Image
     imageType=NoImage
     mirroring=NoMirror
     resourceName=None
     resourceObject=&MI_UI_ForcePower_LightBg
     tiling=NoTile
     imageSize=<32 32>
     margin=<ltrb 0.5 0.5 0.5 0.5>
     outlineSettings {
      bUseBrushTransparency=false
      roundingType=HalfHeightRadius
      width=0
      color {colorUseRule=UseColor_Specified, specifiedColor=<rgba 0 0 0 0>}
      cornerRadii=<0 0 0 0>
     }
     tintColor {colorUseRule=UseColor_Specified, specifiedColor=<rgba 1 1 1 1>}
     uVRegion {bIsValid=false, max=<0 0>, min=<0 0>}
    }
    colorAndOpacity=<rgba 1 1 1 0.2>
    renderTransform {angle=0, scale=<1.7 1.3>, shear=<0 0>, translation=<0 0>}
   }
   slot {class=OverlaySlot, properties {horizontalAlignment=HAlign_Fill, verticalAlignment=VAlign_Fill}}
  }
  Image_ScanLines {
   widgetClassPath=&Image
   bIsVariable=true
   parent=CanPurchase
   widget=Image_ScanLines
   properties {
    brush {
     bIsDynamicallyLoaded=false
     drawAs=Image
     imageType=NoImage
     mirroring=NoMirror
     resourceName=None
     resourceObject=/PlotTwist/Items/RiftGate/Assets/NPCs/RiftGate/UI/Materials/MI_UI_ForcePower_ScanLines
     tiling=NoTile
     imageSize=<32 32>
     margin=<ltrb 0 0 0 0>
     outlineSettings {
      bUseBrushTransparency=false
      roundingType=HalfHeightRadius
      width=0
      color {colorUseRule=UseColor_Specified, specifiedColor=<rgba 0 0 0 0>}
      cornerRadii=<0 0 0 0>
     }
     tintColor {colorUseRule=UseColor_Specified, specifiedColor=<rgba 1 1 1 1>}
     uVRegion {bIsValid=false, max=<0 0>, min=<0 0>}
    }
    colorAndOpacity=<rgba 1 1 1 0.002>
    renderTransform {angle=0, scale=<1.1 0.95>, shear=<0 0>, translation=<0 0>}
   }
   slot {class=OverlaySlot, properties {horizontalAlignment=HAlign_Fill, verticalAlignment=VAlign_Fill}}
  }
  CannotPurchase {
   widgetClassPath=&Overlay
   bIsVariable=true
   parent=VisibilitySwitcher
   widget=CannotPurchase
   slot {class=CommonVisibilitySwitcherSlot, properties {horizontalAlignment=HAlign_Center}}
  }
  BriefErrorMessage {
   widgetClassPath=/FortniteConversationUI/Conversation/BriefErrorMessageNew
   bIsVariable=true
   parent=CannotPurchase
   widget=BriefErrorMessage
   slot {class=OverlaySlot}
  }
 ]
}
