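# M_UI_ForcePower_Icon is a translucent UI material that draws a recolorable Force power icon, defaulting to the Throw icon texture.
# The texture parameter's red, green and blue channels serve as separate masks through named reroutes.
# Two chained lerps layer background, background shape and lightsaber color parameters.
# A separate branch sharpens one mask by multiply and saturate, scales it by an outline transparency parameter, then blends it with summed masks.
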Material /PlotTwist/Items/RiftGate/Assets/NPCs/RiftGate/UI/Materials/M_UI_ForcePower_Icon
graph {
 properties {blendMode=BLEND_Translucent, materialDomain=MD_UI, pixelDepthOffsetMode=PDOM_Legacy, usedShadingModels=Default Lit}
 nodes [
  n1 = NamedRerouteUsage()
  n2 = NamedRerouteUsage()
  n3 = NamedRerouteUsage()
  n4 = Constant() {r=7}
  n5 = Multiply(A=Saturate(None=Multiply(A=@n3, B=@n4)), B=ScalarParameter() {defaultValue=0.8, parameterName=Icon Outline Transparency})
  LinearInterpolate(A=Add(A=Add(A=@n1, B=@n2), B=@n5), Alpha=@n5)
  LinearInterpolate(
   A=LinearInterpolate(
    A=VectorParameter() {parameterName=Bg Color, defaultValue=<rgba 0.1786 0.0089 0.0385 1>}.RGB
    B=VectorParameter() {parameterName=Bg Shape Color, defaultValue=<rgba 1 0 0 1>}.RGB
    Alpha=NamedRerouteUsage()
   )
   B=VectorParameter() {parameterName=Lightsaber Color, defaultValue=<rgba 1 1 1 1>}.RGB
   Alpha=NamedRerouteUsage()
  )
  n6 = TextureSampleParameter2D() {
   parameterName=Force Power Icon
   samplerSource=SSM_Clamp_WorldGroupSettings
   texture=/PlotTwist/Items/RiftGate/Assets/NPCs/RiftGate/UI/Textures/T_UI_Icon_Forcepower_Throw
  }
  NamedRerouteDeclaration(None=@n6.B) {name=BlueChannel, nodeColor=<rgba 0 0.4996 1 1>}
  NamedRerouteDeclaration(None=@n6.G) {name=GreenChannel, nodeColor=<rgba 0 1 0.2529 1>}
  NamedRerouteDeclaration(None=@n6.R) {name=RedChannel, nodeColor=<rgba 1 0 0.2046 1>}
 ]
}
