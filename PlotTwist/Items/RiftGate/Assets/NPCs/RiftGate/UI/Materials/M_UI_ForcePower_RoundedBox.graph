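# Translucent UI material that draws a rounded box with faded sides, named for the Rift Gate force power UI.
# MF_UI_Box supplies the box fill, with roundness set by the Corner_Radius parameter.
# A parabola across the horizontal axis of the second UV channel feeds a SmoothStep, forming the Side Fade Mask.
# The saturated fill is multiplied by a named reroute mask, and a Color parameter is exposed for tinting.
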Material /PlotTwist/Items/RiftGate/Assets/NPCs/RiftGate/UI/Materials/M_UI_ForcePower_RoundedBox
graph {
 properties {blendMode=BLEND_Translucent, materialDomain=MD_UI, pixelDepthOffsetMode=PDOM_Legacy, usedShadingModels=Default Lit}
 nodes [
  n1 = TextureCoordinate() {coordinateIndex=1}
  n2 = Multiply(A=ComponentMask(None=@n1) {r=true}, B=OneMinus(None=ComponentMask(None=@n1) {r=true}))
  Multiply(A=@n2) {constB=3}
  n3 = Constant() {r=1}
  n4 = ScalarParameter() {defaultValue=0.125, group=BackingShape, parameterName=Corner_Radius, sortPriority=0}
  Multiply(
   A=NamedRerouteUsage()
   B=Saturate(None=MaterialFunctionCall(Size=@n3, CornerRoundness=@n4) {materialFunction=&MF_UI_Box}.Fill)
  )
  NamedRerouteDeclaration(None=SmoothStep(Value=@n2) {constMax=0.07}) {name=Side Fade Mask, nodeColor=<rgba 0.0078 1 0 1>}
  VectorParameter() {parameterName=Color, defaultValue=<rgba 1 1 1 0>}
 ]
}
