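# Translucent UI material for the RiftGate Force Power soft box, producing a two-color gradient with an adjustable soft fade.
# A reflected gradient over UI UVs, with parameterized offset, rotation and center, is remapped by width, density and rate into a clamped Fade mask.
# That mask, shared through a named reroute, multiplies a second reflected gradient built from the V coordinate to blend Outer Color into Inner Color.
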Material /PlotTwist/Items/RiftGate/Assets/NPCs/RiftGate/UI/Materials/M_UI_ForcePower_SoftBox
graph {
 properties {blendMode=BLEND_Translucent, materialDomain=MD_UI, pixelDepthOffsetMode=PDOM_Legacy, usedShadingModels=Default Lit}
 nodes [
  n1 = ComponentMask(None=TextureCoordinate()) {g=true}
  n2 = Constant() {r=0.5}
  LinearInterpolate(
   A=VectorParameter() {group=Color, parameterName=Outer Color, sortPriority=1, defaultValue=<rgba 0 0.5235 1 1>}.RGB
   B=VectorParameter() {group=Color, parameterName=Inner Color, sortPriority=1, defaultValue=<rgba 0 0.1026 1 1>}.RGB
   Alpha=Multiply(
    A=MaterialFunctionCall!(UVs=@n1, Offset, RotationCenter, RotationAngle=@n2) {materialFunction=/FortUILibrary/MaterialFunctions/Gradients/MF_UI_ReflectedGradient}.Result
    B=NamedRerouteUsage()
   )
  )
  n3 = ScalarParameter() {defaultValue=0.5, group=FadeTreatment, parameterName=Fade_Rotation_CenterPoint_X, sliderMin=0.5, sortPriority=1}
  n4 = ScalarParameter() {defaultValue=0.5, group=FadeTreatment, parameterName=Fade_Rotation_CenterPoint_Y, sliderMin=0.5, sortPriority=1}
  n5 = ScalarParameter() {defaultValue=1, group=FadeTreatment, parameterName=Fade_Width, sortPriority=1}
  n6 = Subtract(
   A=MaterialFunctionCall(
    UVs=MaterialFunctionCall() {materialFunction=&GetUserInterfaceUV}.Normalized UV
    Offset=ScalarParameter() {group=FadeTreatment, parameterName=Fade_Offset, sortPriority=1}
    RotationCenter=AppendVector(A=@n3, B=@n4)
    RotationAngle=ScalarParameter() {group=FadeTreatment, parameterName=Fade_Rotation, sortPriority=1}
   ) {materialFunction=&MF_UI_ReflectedGradient}.Result
   B=OneMinus(None=@n5)
  ) {constB=0.25}
  n7 = Divide(B=@n5) {constB=0.5}
  n8 = Saturate(None=Multiply(A=Multiply(A=@n6, B=@n7), B=ScalarParameter() {defaultValue=1, group=FadeTreatment, parameterName=Fade_Density, sortPriority=1}))
  n9 = ScalarParameter() {defaultValue=1, group=FadeTreatment, parameterName=Fade_Rate, sortPriority=1}
  NamedRerouteDeclaration(None=Saturate(None=Power(Base=@n8, Exp=@n9))) {name=Fade, nodeColor=<rgba 0 0 0 1>}
  NamedRerouteUsage()
 ]
}
