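# UI material for the Rift Gate conversation panel frame, fitting one channel-packed texture to the widget's aspect ratio.
# Rounded pixel-size ratios scale forward and flipped UVs, whose floored X values mask the first, middle and last horizontal tiles.
# Red fills the middle, while green and blue form the ends, each multiplied by a narrow linear gradient, all merged with Max.
# Derivative-based sampling takes explicit DDX/DDY inputs alongside the wrapped UVs.
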Material /PlotTwist/Items/RiftGate/Assets/NPCs/RiftGate/UI/Materials/M_UI_Frame
graph {
 properties {blendMode=BLEND_Translucent, materialDomain=MD_UI, pixelDepthOffsetMode=PDOM_Legacy, usedShadingModels=Default Lit}
 nodes [
  n1 = NamedRerouteUsage()
  n2 = NamedRerouteUsage()
  n3 = OneMinus(None=@n2)
  n4 = Saturate(None=ComponentMask(None=Floor(None=@n2)) {r=true})
  n5 = TextureSample!(
   UVs=LinearInterpolate(A=Frac(None=@n1), B=Frac(None=@n3), Alpha=@n4)
   Tex
   "DDX(UVs)"=NamedRerouteUsage()
   "DDY(UVs)"=NamedRerouteUsage()
   Apply View MipBias
  ) {
   mipValueMode=TMVM_Derivative
   samplerSource=SSM_Clamp_WorldGroupSettings
   samplerType=SAMPLERTYPE_LinearColor
   texture=/PlotTwist/Items/RiftGate/Assets/NPCs/RiftGate/UI/Textures/T_UI_ConvoPanel_Frame
  }
  n6 = Saturate(None=ComponentMask(None=Floor(None=@n1)) {r=true})
  n7 = Multiply(A=@n6, B=@n4)
  n8 = Multiply(
   A=@n5.G
   B=MaterialFunctionCall(GradientStart=Constant(), GradientEnd=Constant() {r=0.02}) {materialFunction=&MF_UI_LinearGradient}.Result
  )
  n9 = OneMinus(None=@n6)
  n10 = MaterialFunctionCall(GradientStart=Constant(), GradientEnd=Constant() {r=0.02}, RotationAngle=Constant() {r=0.5}) {materialFunction=&MF_UI_LinearGradient}
  Max(A=Max(A=Multiply(A=@n5.R, B=@n7), B=Multiply(A=@n8, B=@n9)), B=Multiply(A=Multiply(A=@n5.B, B=@n10.Result), B=OneMinus(None=@n4)))
  n11 = MaterialFunctionCall() {materialFunction=&GetUserInterfaceUV}
  n12 = ComponentMask(None=@n11.Pixel Size) {r=true}
  n13 = ComponentMask(None=@n11.Pixel Size) {g=true}
  n14 = AppendVector(A=Divide(A=@n12, B=@n13), B=Divide(A=@n13, B=@n12))
  n15 = Constant2Vector() {g=1, r=1}
  NamedRerouteDeclaration(None=Round(None=Max(A=@n14, B=@n15))) {name=UserInterfaceUV, nodeColor=<rgba 0 0 0 1>}
  n16 = TextureCoordinate()
  NamedRerouteDeclaration(None=Multiply(A=@n16, B=NamedRerouteUsage())) {name=UvMultiplied, nodeColor=<rgba 0 0 0 1>}
  NamedRerouteDeclaration(None=Multiply(A=OneMinus(None=@n16), B=NamedRerouteUsage())) {name="UVMultiplied(1-x)", nodeColor=<rgba 0 0 0 1>}
  NamedRerouteDeclaration(None=DDX(Value=@n1)) {name=DDXUV, nodeColor=<rgba 0 0 0 1>}
  NamedRerouteDeclaration(None=DDY(Value=@n1)) {name=DDYUV, nodeColor=<rgba 0 0 0 1>}
 ]
}
